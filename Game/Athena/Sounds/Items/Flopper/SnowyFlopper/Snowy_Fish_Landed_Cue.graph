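# Sound cue for the Snowy Flopper fish item landing, routed through the loot-collect sound class with weapon-drop attenuation.
# A mixer layers the snowy-specific landing wave over the first generic Flopper landed wave, with the generic one slightly quieter.
# The mix feeds a root modulator whose volume range is fixed, and the cue plays at reduced overall volume.
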SoundCue /Game/Athena/Sounds/Items/Flopper/SnowyFlopper/Snowy_Fish_Landed_Cue
root=Modulator_1
nodes [
 {id=Mixer_0, children [WavePlayer_1,WavePlayer_0], properties {inputVolume [0.9,1]}}
 {id=Modulator_1, children [Mixer_0], properties {volumeMax=1, volumeMin=1}}
 {id=WavePlayer_0, properties {soundWaveAssetPtr=&/Snowy_Fish_Landed}}
 {id=WavePlayer_1, properties {soundWaveAssetPtr=/Game/Athena/Sounds/Items/Flopper/Flopper_Landed_01}}
]
properties {
 attenuationSettings=&Fort_Weapon_Drop_Att
 maxDistance=3000
 soundClassObject=&Fort_Loot_Collect
 volumeMultiplier=0.6
}
