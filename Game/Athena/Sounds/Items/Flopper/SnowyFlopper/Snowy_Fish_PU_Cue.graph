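# Sound cue for picking up the Snowy Flopper fish item, judging by its name and loot collect sound class.
# A single wave player plays the Snowy_Fish_PU wave as the root, with no randomization or modulation.
# It uses the shared weapon drop attenuation settings to localize the sound in the world.
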SoundCue /Game/Athena/Sounds/Items/Flopper/SnowyFlopper/Snowy_Fish_PU_Cue
root=WavePlayer_7
nodes [{id=WavePlayer_7, properties {soundWaveAssetPtr=&/Snowy_Fish_PU}}]
properties {
 attenuationSettings=&Fort_Weapon_Drop_Att
 maxDistance=3000
 soundClassObject=&Fort_Loot_Collect
}
